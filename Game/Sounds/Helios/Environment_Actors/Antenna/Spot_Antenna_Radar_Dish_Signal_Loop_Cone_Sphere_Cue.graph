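# Looping ambient spot sound cue for radar dish signals on an antenna, classed as outdoor ambience and sent to the exterior ambient source bus.
# A Random node picks one of five looping dish-signal waves, which passes through slight downward pitch modulation and a boosting Mixer.
# The Quality Level root feeds that chain to its first two inputs and leaves the others empty.
# Attenuation references a shared radar dish asset, and overridden concurrency allows one instance.
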SoundCue /Game/Sounds/Helios/Environment_Actors/Antenna/Spot_Antenna_Radar_Dish_Signal_Loop_Cone_Sphere_Cue
root=QualityLevel_0
nodes [
 {id=Mixer_0, children [Modulator_0], properties {inputVolume [1.65]}}
 {id=Modulator_0, children [Random_0], properties {pitchMax=1, pitchMin=0.8, volumeMax=1, volumeMin=1}}
 {id=QualityLevel_0, children [Mixer_0,Mixer_0,null,null]}
 {id=Random_0, children [WavePlayer_0,WavePlayer_1,WavePlayer_3,WavePlayer_5,WavePlayer_6], properties {weights [1,1,1,1,1]}}
 {id=WavePlayer_0, properties {bLooping=true, soundWaveAssetPtr=/Game/Sounds/Apollo/Biomes/RadarBases/IOAgents_DishSignals_01}}
 {id=WavePlayer_1, properties {bLooping=true, soundWaveAssetPtr=/Game/Sounds/Apollo/Biomes/RadarBases/IOAgents_DishSignals_03}}
 {id=WavePlayer_3, properties {bLooping=true, soundWaveAssetPtr=/Game/Sounds/Apollo/Biomes/RadarBases/IOAgents_DishSignals_04}}
 {id=WavePlayer_5, properties {bLooping=true, soundWaveAssetPtr=/Game/Sounds/Apollo/Biomes/RadarBases/IOAgents_DishSignals_06}}
 {id=WavePlayer_6, properties {bLooping=true, soundWaveAssetPtr=/Game/Sounds/Apollo/Biomes/RadarBases/IOAgents_DishSignals_07}}
]
properties {
 attenuationSettings=/Game/Sounds/Asteria/POI/ExplorerBase/Spots/Spot_ExplorerBase_Radar_Dish_Smaller_Att
 bOverrideConcurrency=true
 maxDistance=1500
 soundClassObject=&Fort_Ambient_Outdoors_Athena
 volumeMultiplier=0.301
 attenuationOverrides {
  bAttenuateWithLPF=true
  bEnableReverbSend=false
  falloffDistance=7000
  lPFFrequencyAtMax=500
  lPFRadiusMax=5000
  lPFRadiusMin=2000
  stereoSpread=5000
  attenuationShapeExtents {x=1000}
 }
 busSends [{audioBus=null, minSendLevel=0.6, sendLevel=1, soundSourceBus=&AmbientAudio_Exterior_SourceBus}]
 concurrencyOverrides {maxCount=1}
}
